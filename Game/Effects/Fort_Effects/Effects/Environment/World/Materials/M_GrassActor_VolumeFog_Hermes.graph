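# Additive volume-domain material for grass actor fog, usable on instanced static meshes.
# Two world-space samples of the VT_ClothSimWPO texture scroll over time and blend into noise that breaks up density and offsets the height cutoff.
# Density multiplies that noise by a radial falloff within object bounds, a soft instance-space vertical fade, a scalar and a camera-depth fade.
# Several leftover nodes, including a sky light color multiply, have missing inputs.
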Material /Game/Effects/Fort_Effects/Effects/Environment/World/Materials/M_GrassActor_VolumeFog_Hermes
graph {
 properties {
  bAutomaticallySetUsageInEditor=false
  bUsedWithInstancedStaticMeshes=true
  blendMode=BLEND_Additive
  materialDomain=MD_Volume
  usedShadingModels=Default Lit
 }
 nodes [
  ActorPositionWS()
  ActorPositionWS()
  n1 = Add(A=WorldPosition().XYZ, B=Multiply(A=Time(), B=VectorParameter() {parameterName=Param_1, defaultValue=<rgba 10 0 50 0>}.RGB))
  n2 = ScalarParameter() {defaultValue=0.0005, parameterName=Param_2}
  n3 = Add(A=WorldPosition().XYZ, B=Multiply(A=Time(), B=VectorParameter() {parameterName=Noise, defaultValue=<rgba 20 0 -50 0>}.RGB))
  n4 = ScalarParameter() {defaultValue=0.001, parameterName=secondary noise}
  n5 = LinearInterpolate(
   A=TextureSample(UVs=Multiply(A=@n1, B=@n2)) {samplerType=SAMPLERTYPE_LinearColor, texture=/Game/Global/MaterialFunctions/VT_ClothSimWPO}.RGB
   B=TextureSample(UVs=Multiply(A=@n3, B=@n4)) {samplerType=SAMPLERTYPE_LinearColor, texture=&VT_ClothSimWPO}.RGB
  )
  Add(A=Multiply(A=@n5, B=ScalarParameter() {defaultValue=1, parameterName=Param_4}))
  Add()
  ComponentMask() {g=true, r=true}
  ComponentMask() {b=true}
  MaterialFunctionCall() {materialFunction=&MF_FX_EmissiveAdjustments}
  MaterialFunctionCall() {materialFunction=&MF_FX_EmissiveAdjustments}
  n6 = TransformPosition(None=WorldPosition().XYZ) {transformSourceType=TRANSFORMPOSSOURCE_World}
  n7 = ObjectLocalBounds()
  n8 = Distance(A=Divide(A=Subtract(A=@n6, B=@n7.Min), B=@n7.Extents), B=Constant() {r=0.5})
  n9 = ScalarParameter() {defaultValue=-0.25, parameterName=Param_5}
  n10 = ScalarParameter() {defaultValue=1}
  n11 = Multiply(A=OneMinus(None=Multiply(A=@n8) {constB=2}), B=Saturate(None=LinearInterpolate(A=@n9, B=@n10, Alpha=@n5)))
  n12 = ObjectPositionWS()
  n13 = Add(A=Multiply(A=@n5, B=ScalarParameter() {defaultValue=75, parameterName=Param_6}), B=WorldPosition().XYZ)
  n14 = ComponentMask(None=TransformPosition(None=Subtract(A=@n12, B=@n13)) {transformSourceType=TRANSFORMPOSSOURCE_World, transformType=TRANSFORMPOSSOURCE_Instance}) {b=true}
  n15 = ScalarParameter() {defaultValue=-75, parameterName=offset}
  n16 = Saturate(None=Divide(A=Subtract(A=@n14, B=@n15), B=ScalarParameter() {defaultValue=75, parameterName=soften}))
  n17 = Divide(A=PixelDepth().R, B=ScalarParameter() {defaultValue=5000, parameterName=Camera distance})
  Multiply(
   A=Multiply(A=Multiply(A=@n11, B=@n16), B=ScalarParameter() {defaultValue=0.05, parameterName=density scalar})
   B=OneMinus(None=Saturate(None=@n17))
  )
  n18 = CollectionParameter() {
   collection=&FortniteMaterialParameters
   parameterName=SkyLightColorAndIntensity
  }
  Multiply(B=ComponentMask(None=@n18) {b=true, g=true, r=true})
  Multiply()
  Multiply()
  Multiply()
  Noise!(World Position, FilterWidth) {noiseFunction=NOISEFUNCTION_GradientTex, outputMax=2, outputMin=0}
  OneMinus(None=@n16)
  ScalarParameter() {defaultValue=2500, parameterName=Param_1}
  TransformPosition() {transformSourceType=TRANSFORMPOSSOURCE_World, transformType=TRANSFORMPOSSOURCE_Instance}
  VectorParameter() {defaultValue=<rgba 1 1 1 1>}
 ]
}
