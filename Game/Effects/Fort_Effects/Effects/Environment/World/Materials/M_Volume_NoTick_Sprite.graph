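# Additive, two-sided material for instanced meshes that draws soft, drifting cloud-like volume sprites, animated entirely in the shader.
# Two panning storm-cloud noise samples distort a radial mask, with depth, camera-proximity, lifetime and cull-distance fades.
# The Sprite function builds billboards through world position offset; time plus vertex alpha loops their drift and life.
# Vertex red varies sprite size; green varies pan speed and opacity.
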Material /Game/Effects/Fort_Effects/Effects/Environment/World/Materials/M_Volume_NoTick_Sprite
graph {
 properties {
  bAutomaticallySetUsageInEditor=false
  bUsedWithInstancedStaticMeshes=true
  blendMode=BLEND_Additive
  twoSided=true
  usedShadingModels=Default Lit
 }
 nodes [
  n1 = TextureCoordinate()
  n2 = ScalarParameter() {defaultValue=0.75, parameterName=Large Noise Size}
  n3 = Time()
  n4 = Multiply(
   A=LinearInterpolate(Alpha=VertexColor().G) {constA=1, constB=2}
   B=ScalarParameter() {defaultValue=-0.025, parameterName=Large Noise Pan Speed}
  )
  n5 = TextureSample(UVs=Panner(Coordinate=Multiply(A=@n1, B=@n2), Time=Multiply(A=@n3, B=@n4)) {speedY=1}) {samplerType=SAMPLERTYPE_LinearColor, texture=/Game/VisualThreatManager/StormVisuals/Textures/Clouds_VectorDisplacement}
  n6 = LinearInterpolate(Alpha=VertexColor().G) {constA=1, constB=2}
  n7 = ScalarParameter() {defaultValue=-0.005, parameterName=Param_6}
  n8 = Panner(
   Coordinate=Multiply(A=TextureCoordinate(), B=ScalarParameter() {defaultValue=0.25, parameterName=Param_5})
   Time=Multiply(A=Time(), B=Multiply(A=@n6, B=@n7))
  ) {speedY=1}
  n9 = Power(
   Base=Add(
    A=Power(Base=@n5.R)
    B=TextureSample(UVs=@n8) {samplerType=SAMPLERTYPE_LinearColor, texture=&Clouds_VectorDisplacement}.G
   )
   Exp=ScalarParameter() {defaultValue=2, parameterName=Texture Power}
  )
  n10 = ScalarParameter() {defaultValue=0.01, parameterName=Distortion Scale}
  n11 = SphereMask(
   A=Add(A=Multiply(A=@n9, B=@n10), B=TextureCoordinate())
   B=ScalarParameter() {defaultValue=0.5, parameterName=Param_2}
   Radius=ScalarParameter() {defaultValue=0.4296, parameterName=SphereMask Scale}
  ) {hardnessPercent=0}
  n12 = ScalarParameter() {defaultValue=1250, parameterName=Camera Fade Distance}
  n13 = Multiply(
   A=DepthFade(Opacity=Multiply(A=@n9, B=@n11), FadeDistance=ScalarParameter() {defaultValue=350, parameterName=Depth Fade})
   B=Power(
    Base=MaterialFunctionCall(Distance=@n12) {materialFunction=&Near_Camera_Fade}.Result
    Exp=ScalarParameter() {defaultValue=8, parameterName=Fade Distance Power}
   )
  )
  n14 = NamedRerouteUsage()
  n15 = NamedRerouteUsage()
  n16 = ScalarParameter() {defaultValue=1, parameterName=Cull Distance Opacity Power}
  Multiply(
   A=Multiply(A=Multiply(A=@n13, B=@n14), B=Power(Base=@n15, Exp=@n16) {constExponent=4})
   B=LinearInterpolate(
    A=ScalarParameter() {defaultValue=0.05, parameterName=Random Opacity Scale Low}
    B=ScalarParameter() {defaultValue=0.15, parameterName=Random Opacity Scale High}
    Alpha=VertexColor().G
   )
  )
  n17 = WorldPosition()
  n18 = TextureCoordinate()
  n19 = VertexColor()
  n20 = ScalarParameter() {defaultValue=750, parameterName=Param_72}
  n21 = Multiply(
   A=LinearInterpolate(A=ScalarParameter() {defaultValue=1000, parameterName=Param_1}, B=ScalarParameter() {defaultValue=2250}, Alpha=@n19.R)
   B=Power(
    Base=MaterialFunctionCall(Distance=@n20) {materialFunction=&Near_Camera_Fade}.Result
    Exp=ScalarParameter() {defaultValue=4, parameterName=Param_9}
   )
  )
  n22 = Time()
  n23 = ScalarParameter() {defaultValue=0.25, parameterName=Param_11}
  n24 = Fmod(A=Add(A=@n19.A, B=Multiply(A=@n22, B=@n23)), B=ScalarParameter() {defaultValue=1, parameterName=Param_10})
  n25 = VectorParameter() {parameterName=Param_1, defaultValue=<rgba 500 0 0 0>}
  Multiply(
   A=Add(
    A=MaterialFunctionCall!(Center Location=@n17.XYZ, 0-1 UVs=@n18, XY scale=@n21, OverridePivot, "Normals (Optional)") {materialFunction=&Sprite}.World position offset
    B=Multiply(A=@n24, B=@n25.RGB)
   )
   B=@n15
  )
  NamedRerouteDeclaration(None=SphereMask(
    A=@n24
    B=ScalarParameter() {defaultValue=0.5, parameterName=Param_12}
    Radius=ScalarParameter() {defaultValue=0.5, parameterName=Param_13}
    Hardness=ScalarParameter() {parameterName=Param_14}
   )) {name=Opacity over Life, nodeColor=<rgba 0.125 1 0 1>}
  NamedRerouteDeclaration(None=SphereMask(A=WorldPosition().XYZ, B=CameraPositionWS(), Radius=ScalarParameter() {defaultValue=4000, parameterName=Param_15}) {hardnessPercent=0}) {name=Cull Distance Fade, nodeColor=<rgba 0 0.9531 1 1>}
  SphereMask(A=WorldPosition().XYZ, B=CameraPositionWS(), Radius=ScalarParameter() {defaultValue=4000, parameterName=Param_15}) {hardnessPercent=0}
  VectorParameter() {defaultValue=<rgba 1 1 1 0>}
 ]
}
